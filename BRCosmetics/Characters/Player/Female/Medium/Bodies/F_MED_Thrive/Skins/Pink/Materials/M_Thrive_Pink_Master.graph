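# Master material for the Pink skin of the Thrive character, adding glowing animated tendrils and Fresnel glow over standard character FX functions.
# Panned, rotated noise with curl-noise distortion forms tendrils, while FX-texture and UV masks separate sleeves, legs, stockings and skirt.
# Static switches gate tendril color, emissive and normal-based vertex offset, plus optional base glow.
# A reactive mask blends previous to target intensity from a Blueprint timestamp, then decays.
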Material /BRCosmetics/Characters/Player/Female/Medium/Bodies/F_MED_Thrive/Skins/Pink/Materials/M_Thrive_Pink_Master
graph {
 properties {
  bAutomaticallySetUsageInEditor=false
  bUseMaterialAttributes=true
  bUsedWithMeshDeformer=true
  bUsedWithMorphTargets=true
  bUsedWithSkeletalMesh=true
  pixelDepthOffsetMode=PDOM_Legacy
  refractionMethod=RM_IndexOfRefraction
  shadingModel=MSM_FromMaterialExpression
  subsurfaceProfile=&SS_HeroSkin_02
  usedShadingModels=Default Lit | Subsurface | Clear Coat | Subsurface Profile | Cloth | Eye
 }
 nodes [
  n1 = MaterialFunctionCall() {materialFunction=&BaseFN_CharacterFunction_PreFX}
  n2 = GetMaterialAttributes(None=MaterialFunctionCall(
    PreskinnedPosition=NamedRerouteDeclaration(None=@n1.Pre skinned Local Position) {name=PreSkinned_Local_Pos, nodeColor=<rgba 0.2188 0 1 1>}
    Gmap=@n1.Gmap
    MaterialAttributes=@n1.Gmap
    AO=@n1.Gmap
   ) {materialFunction=&CharacterSkinFX}.Result) {attributeGetTypes [#1]}
  n3 = NamedRerouteUsage()
  n4 = TextureCoordinate()
  n5 = MaterialFunctionCall(Float2=@n4) {materialFunction=&BreakOutFloat2Components}
  n6 = ComponentMask(None=NamedRerouteUsage()) {r=true}
  n7 = LinearInterpolate(
   A=@n4
   B=MaterialFunctionCall(X=Multiply(A=@n5.R) {constB=-1}, Y=@n5.G) {materialFunction=&MakeFloat2}.Result
   Alpha=Saturate(None=Sign(None=@n6))
  )
  n8 = Constant() {r=0.35}
  n9 = Multiply(A=Multiply(
    A=MaterialFunctionCall(UVs=@n7, "Rotation Angle (0-1)"=@n8) {materialFunction=&CustomRotator}.Rotated Values
    B=Constant2Vector() {g=3, r=0.1}
   )) {constB=4}
  n10 = ComponentMask(None=VectorParameter() {group=Thrive_Pink, parameterName=Leg_ScrollSpeed, defaultValue=<rgba -0.15 0.1 0 0>}.RGB) {g=true, r=true}
  n11 = NamedRerouteUsage()
  n12 = Multiply(A=Time(), B=ScalarParameter() {defaultValue=-0.1, parameterName=Tendril_Distortion_Speed})
  n13 = TextureSample(UVs=Add(
    A=Divide(A=@n11) {constB=200}
    B=MaterialFunctionCall(Z=@n12) {materialFunction=&MakeFloat3}.Result
   )) {samplerSource=SSM_Wrap_WorldGroupSettings, samplerType=SAMPLERTYPE_LinearColor, texture=/Game/Effects/FogVolume/VT_CurlNoise_LOW}
  n14 = Multiply(A=ComponentMask(None=@n13.RGB) {b=true, g=true}) {constB=0.0075}
  n15 = TextureSample(UVs=Add(A=Panner(Coordinate=@n9, Speed=@n10) {speedX=1, speedY=1}, B=@n14)) {
   samplerSource=SSM_Wrap_WorldGroupSettings
   samplerType=SAMPLERTYPE_LinearGrayscale
   texture=/BRCosmetics/Athena/Effects/Particles/Skydiving/NS_Contrail_DarkAndLight/T_DarkAndLight_Noise
  }
  n16 = TextureCoordinate()
  n17 = Constant() {r=0.45}
  n18 = Multiply(A=Multiply(
    A=MaterialFunctionCall(UVs=@n16, "Rotation Angle (0-1)"=@n17) {materialFunction=&CustomRotator}.Rotated Values
    B=Constant2Vector() {g=3, r=0.1}
   )) {constB=4}
  n19 = ComponentMask(None=VectorParameter() {group=Thrive_Pink, parameterName=Sleeve_ScrollSpeed, defaultValue=<rgba 0.075 0.05 0 0>}.RGB) {g=true, r=true}
  n20 = TextureSample(UVs=Add(A=Panner(Coordinate=@n18, Speed=@n19) {speedX=1, speedY=1}, B=@n14)) {
   samplerSource=SSM_Wrap_WorldGroupSettings
   samplerType=SAMPLERTYPE_LinearGrayscale
   texture=&T_DarkAndLight_Noise
  }
  n21 = DotProduct(A=ComponentMask(None=VertexNormalWS()) {b=true}, B=Constant() {r=-1})
  n22 = Multiply(A=LinearInterpolate(A=@n15.G, B=@n20.G, Alpha=NamedRerouteUsage()) {constA=0.33, constB=-0.33}, B=OneMinus(None=Saturate(None=@n21)))
  n23 = ScalarParameter() {defaultValue=0.7, group=Thrive_Pink, parameterName=Tendril_Clip_Amount}
  n24 = Saturate(None=MaterialFunctionCall!(In=Subtract(A=@n22, B=@n23) {constB=0.7}, Ramp Width=Constant() {r=0.25}) {materialFunction=/Engine/Functions/Engine_MaterialFunctions01/Math/SmoothCeil}.Result)
  n25 = Add(A=Add(A=NamedRerouteDeclaration(None=@n24) {name=Tendril_Mask, nodeColor=<rgba 0.6406 0 1 1>}) {constB=0.2}, B=NamedRerouteUsage())
  n26 = NamedRerouteUsage()
  n27 = Multiply(A=Add(A=Multiply(B=@n3) {constA=0.5}, B=Multiply(A=@n25, B=@n26)), B=NamedRerouteUsage())
  n28 = NamedRerouteUsage()
  n29 = Add(A=NamedRerouteUsage(), B=Multiply(A=NamedRerouteUsage(), B=NamedRerouteUsage()))
  n30 = Add(
   A=@n27
   B=Multiply(A=Multiply(A=@n28, B=@n29), B=ScalarParameter() {defaultValue=0.3, group=Thrive_Pink, parameterName=SleeveBar_BaseNoise_Intensity})
  )
  n31 = LinearInterpolate(
   A=@n2.Base Color
   B=VectorParameter() {group=Thrive_Pink, parameterName=Tendril_Base_Color, defaultValue=<rgba 1 0.0938 0.9443 0>}.RGB
   Alpha=Saturate(None=StaticSwitchParameter(True=@n30, False=@n27) {group=Thrive_Pink, parameterName=Add_Base_Glow})
  )
  n32 = GetMaterialAttributes(None=SetMaterialAttributes!(
    MaterialAttributes=@n2.MaterialAttributes
    Base Color=StaticSwitchParameter(True=@n31, False=@n2.Base Color) {defaultValue=true, group=Thrive_Pink, parameterName=Add_Tendrils}
   ) {attributeSetTypes [#1]}) {attributeGetTypes [#2]}
  n33 = SetMaterialAttributes!(MaterialAttributes=@n32.MaterialAttributes, Emissive Color=Add(A=@n32.Emissive Color, B=NamedRerouteUsage()) {constB=0}) {attributeSetTypes [#2]}
  n34 = GetMaterialAttributes(None=@n33) {attributeGetTypes [#3]}
  n35 = NamedRerouteUsage()
  n36 = NamedRerouteUsage()
  n37 = VertexNormalWS()
  n38 = ScalarParameter() {defaultValue=0.4, parameterName=Pink_WPO_Amount}
  n39 = Add(A=@n34.World Position Offset, B=Multiply(A=Multiply(A=@n35, B=@n36), B=Multiply(A=@n37, B=@n38))) {constB=0}
  n40 = MaterialFunctionCall(
   AO=@n1.AO
   Distance Blend=@n1.AO
   Fresnel=@n1.AO
   Gmap=@n1.AO
   Pre Skinned Local Position=@n1.AO
   Material=SetMaterialAttributes!(
    MaterialAttributes=@n34.MaterialAttributes
    World Position Offset=StaticSwitchParameter(True=@n39, False=@n34.World Position Offset) {defaultValue=true, group=Thrive_Pink, parameterName=Add_Tendrils}
   ) {attributeSetTypes [#3]}
  ) {materialFunction=&BaseFN_CharacterFunction_PostFX}
  MaterialFunctionCall(In=@n40.Result) {materialFunction=&CINE_MF_CharPostFX}
  n41 = SphereMask(
   A=TextureCoordinate()
   B=ComponentMask(None=VectorParameter() {parameterName=Sleeve_Mask_Center, defaultValue=<rgba 0.45 0.47 0 0>}.RGB) {g=true, r=true}
   Radius=Constant() {r=0.275}
   Hardness=Constant() {r=0.75}
  )
  n42 = TextureSample() {
   samplerType=SAMPLERTYPE_Masks
   texture=/BRCosmetics/Characters/Player/Female/Medium/Bodies/F_MED_Thrive/Skins/Pink/Textures/T_F_MED_Thrive_Pink_Body_FX
  }
  n43 = Subtract(A=@n42.B) {constB=0.2}
  n44 = Ceil(None=Subtract(A=Add(A=@n42.R, B=@n43), B=@n42.G))
  n45 = NamedRerouteDeclaration(None=Multiply(A=Saturate(None=@n41), B=Saturate(None=@n44))) {name=Mask_Sleeve_Only, nodeColor=<rgba 1 0 0.4453 1>}
  n46 = Ceil(None=Subtract(A=@n42.R) {constB=0.15})
  n47 = Add(A=@n42.RGB) {constB=0.6}
  n48 = Add(A=OneMinus(None=Floor(None=@n47)), B=Floor(None=Add(A=@n42.RGB) {constB=0.45}))
  n49 = MaterialFunctionCall(Float3=OneMinus(None=Saturate(None=@n48))) {materialFunction=&BreakOutFloat3Components}
  n50 = Add(A=@n46, B=@n49.B)
  n51 = Multiply(
   A=Floor(None=OneMinus(None=@n45))
   B=NamedRerouteDeclaration(None=Saturate(None=@n50)) {name=Mask_LegsAndSleeves, nodeColor=<rgba 0 1 0.0156 1>}
  )
  n52 = TextureCoordinate()
  n53 = ComponentMask(None=@n52) {g=true}
  n54 = Ceil(None=Saturate(None=Subtract(A=@n53) {constB=0.65}))
  n55 = Abs(None=Add(A=ComponentMask(None=@n52) {r=true}) {constB=-0.5})
  n56 = Ceil(None=Saturate(None=Subtract(A=@n55) {constB=0.29}))
  NamedRerouteDeclaration(None=Multiply(A=NamedRerouteDeclaration(None=@n51) {name=Mask_LegsAndGloves_Only, nodeColor=<rgba 0.8281 1 0 1>}, B=Multiply(A=@n54, B=@n56))) {name=Mask_Stockings, nodeColor=<rgba 0 1 0.7656 1>}
  n57 = NamedRerouteUsage()
  n58 = Add(A=@n57)
  n59 = Saturate(None=Fresnel(ExponentIn=LinearInterpolate(Alpha=@n57) {constA=1.5, constB=7}) {baseReflectFraction=0, exponent=2})
  n60 = Multiply(
   A=NamedRerouteDeclaration(None=Multiply(A=@n58, B=@n59)) {name=Mask_Fresnel, nodeColor=<rgba 1 0 0.375 1>}
   B=VectorParameter() {group=Thrive_Pink, parameterName=Pink_Fresnel_Glow, defaultValue=<rgba 2 0.0833 1.5037 0>}.RGB
  )
  n61 = Multiply(A=Fresnel(Normal=VertexNormalWS()) {baseReflectFraction=0.15, exponent=0.75}, B=Multiply(A=NamedRerouteUsage(), B=NamedRerouteUsage()))
  n62 = TextureCoordinate()
  n63 = TextureSample(UVs=Panner(Coordinate=Multiply(A=@n62) {constB=2}) {speedY=-0.1}) {
   samplerSource=SSM_Wrap_WorldGroupSettings
   samplerType=SAMPLERTYPE_LinearGrayscale
   texture=&T_DarkAndLight_Noise
  }
  n64 = Multiply(A=@n63.G) {constB=2}
  n65 = VectorParameter() {group=Thrive_Pink, parameterName=Pink_Tendril_Glow, defaultValue=<rgba 0.8232 0.1142 0.9531 0>}
  n66 = Multiply(
   A=LinearInterpolate(A=Saturate(None=@n61), B=NamedRerouteDeclaration(None=@n64) {name=Mask_BaseNoise, nodeColor=<rgba 0.5938 0 1 1>}, Alpha=NamedRerouteUsage()) {constB=0}
   B=@n65.RGB
  )
  n67 = Multiply(
   A=NamedRerouteUsage()
   B=StaticSwitchParameter(True=Add(A=@n60, B=@n66) {constB=0}, False=@n60) {defaultValue=true, group=Thrive_Pink, parameterName=Add_Tendrils}
  )
  n68 = NamedRerouteUsage()
  n69 = Add(A=NamedRerouteUsage(), B=Multiply(A=NamedRerouteUsage(), B=NamedRerouteUsage()))
  n70 = Multiply(
   A=@n65.RGB
   B=Multiply(A=Multiply(A=@n68, B=@n69), B=ScalarParameter() {defaultValue=0.3, group=Thrive_Pink, parameterName=SleeveBar_BaseNoise_Intensity})
  )
  NamedRerouteDeclaration(None=StaticSwitchParameter(True=Add(A=@n67, B=@n70), False=@n67) {group=Thrive_Pink, parameterName=Add_Base_Glow}) {name=Emissive_Out, nodeColor=<rgba 1 0 0.9844 1>}
  NamedRerouteDeclaration(None=TextureSampleParameter2D() {parameterName=M, samplerSource=SSM_Wrap_WorldGroupSettings}.B) {name=Mask_Skin_Only, nodeColor=<rgba 0 1 0.0859 1>}
  NamedRerouteDeclaration(None=Saturate(None=@n46)) {name=Mask_SleeveBars, nodeColor=<rgba 0 1 0.6016 1>}
  n71 = Subtract(A=Time(), B=ScalarParameter() {defaultValue=-1000, parameterName=Thrive_GameTimeFromBP})
  n72 = ScalarParameter() {defaultValue=10, parameterName=Thrive_EffectDecayRate}
  n73 = OneMinus(None=Saturate(None=Divide(A=@n71, B=@n72)))
  n74 = ScalarParameter() {defaultValue=0.75, group=Thrive_Pink, parameterName=Thrive_TransitionTime}
  n75 = LinearInterpolate(
   A=ScalarParameter() {parameterName=Thrive_Previous_Intensity}
   B=ScalarParameter() {parameterName=Thrive_Target_Intensity}
   Alpha=Saturate(None=Divide(A=@n71, B=@n74) {constB=0.75})
  )
  NamedRerouteDeclaration(None=LinearInterpolate(
    A=Multiply(A=@n73, B=@n75)
    B=ScalarParameter() {defaultValue=1, group=Thrive_Pink, parameterName=Debug_Reactive_Level}
    Alpha=ScalarParameter() {parameterName=FE_Preview}
   )) {name=Mask_Reactive, nodeColor=<rgba 0 1 0.3906 1>}
  NamedRerouteDeclaration(None=@n49.G) {name=Mask_Skirt, nodeColor=<rgba 0.4062 1 0 1>}
  n76 = GetMaterialAttributes(None=@n33) {attributeGetTypes [#4]}
  SetMaterialAttributes!(MaterialAttributes=@n76.MaterialAttributes, Normal=@n76.MaterialAttributes) {attributeSetTypes [#4]}
  StaticSwitchParameter() {group=Thrive_Pink, parameterName=Debug_Preview_Full_Power}
  n77 = NamedRerouteUsage()
  n78 = NamedRerouteUsage()
  n79 = Multiply(A=NamedRerouteUsage(), B=Add(A=NamedRerouteUsage(), B=Multiply(A=@n77, B=@n78)))
  StaticSwitchParameter(True=LinearInterpolate(A=@n76.Normal, B=Constant3Vector() {constant=<rgba 0 0 1 0>}, Alpha=Saturate(None=@n79)), False=@n76.Normal) {defaultValue=true, group=Thrive_Pink, parameterName=Add_Tendrils}
  TextureSampleParameter2D() {parameterName=SkinFX_Mask, samplerSource=SSM_Wrap_WorldGroupSettings}
 ]
}
